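# Translucent unlit master material for smoke strips on Niagara mesh particles, driven by dynamic parameters.
# Two distorted, panning smoke textures set color, tinted by particle and time-of-day colors with optional curve-atlas gradient.
# Their opacity is shaped by procedural fire-shape and bottom-fade masks, depth and view-angle fades, and an optional particle-centered sphere mask.
# World position offset combines scrolling noise and sine wobble with switchable SplineThicken and clamped bend interpolation.
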Material /BRCosmetics/Effects/Niagara/MasterMaterials/Smoke/M_SmokeStrip
graph {
 properties {
  bUsedWithNiagaraMeshParticles=true
  blendMode=BLEND_Translucent
  pixelDepthOffsetMode=PDOM_Legacy
  shadingModel=MSM_Unlit
  twoSided=true
  usedShadingModels=Unlit
 }
 nodes [
  n1 = ScalarParameter() {parameterName=SplineThicken_WidthBase}
  n2 = VertexColor()
  n3 = MaterialFunctionCall(Float2=TextureCoordinate()) {materialFunction=&BreakOutFloat2Components}
  n4 = ScalarParameter() {defaultValue=3, parameterName=SplineThicken_WidthBaseFade_Power}
  n5 = MaterialFunctionCall(
   WidthBase=LinearInterpolate(A=Multiply(A=@n1, B=@n2.A), Alpha=Power(Base=@n3.G, Exp=@n4)) {constB=0}
   WidthTip=Multiply(A=ScalarParameter() {parameterName=SplineThicken_WidthTop}, B=@n2.A)
  ) {materialFunction=&SplineThicken}
  n6 = DynamicParameter() {desc=Parameters set by Niagara, parameterIndex=1, defaultValue=<rgba 1 1 50 1>, paramNames [WPO UV Scale, "", "", ""]}
  n7 = MaterialFunctionCall(X=@n6.WPO UV Scale, Y=@n6.WPO UV Scale, Z=@n6.WPO UV Scale) {materialFunction=&MakeFloat3}
  n8 = ParticlePositionWS()
  n9 = ScalarParameter() {defaultValue=20, group=WPO, parameterName=Interpolate Clamp}
  n10 = ComponentMask(None=OneMinus(None=TextureCoordinate())) {g=true}
  n11 = ScalarParameter() {defaultValue=2, group=WPO, parameterName=WPOBendPower}
  n12 = LinearInterpolate(
   B=Clamp(None=Subtract(A=@n7.Result, B=@n8), Min=Multiply(B=@n9) {constA=-1}, Max=@n9) {maxDefault=1000, minDefault=-1000}
   Alpha=Multiply(A=Power(Base=@n10, Exp=@n11), B=@n6.A)
  )
  n13 = StaticSwitchParameter(True=@n12, False=Constant()) {defaultValue=true, group=WPO, parameterName=Use_Interpolate WPO}
  n14 = Add(A=TextureCoordinate(), B=ParticleRandom())
  n15 = ComponentMask(None=VectorParameter() {group=WPO, parameterName=WPO_B_UVScale, defaultValue=<rgba 0 1 0 0>}.RGB) {g=true, r=true}
  n16 = AppendVector(A=Constant(), B=ScalarParameter() {defaultValue=0.5, group=WPO, parameterName=WPO_B_Speed})
  n17 = DynamicParameter() {defaultValue=<rgba 1 1 50 1>, paramNames [Smoke Pan Speed, WPO Speed, Sphere Mask Radius, WPO Intensity]}
  n18 = TextureSample(UVs=Panner(Coordinate=Multiply(A=@n14, B=@n15), Speed=Multiply(A=@n16, B=@n17.WPO Speed)) {speedY=0.5}) {samplerType=SAMPLERTYPE_Normal, texture=&T_Weap_Distort_N_02}
  n19 = Add(A=TextureCoordinate(), B=ParticleRandom())
  n20 = ComponentMask(None=VectorParameter() {group=WPO, parameterName=WPO_A_UVScale, defaultValue=<rgba 0 1 0 0>}.RGB) {g=true, r=true}
  n21 = AppendVector(A=Constant(), B=ScalarParameter() {defaultValue=0.5, group=WPO, parameterName=WPO_A_Speed})
  n22 = TextureSample(UVs=Panner(Coordinate=Multiply(A=@n19, B=@n20), Speed=Multiply(A=@n21, B=@n17.WPO Speed)) {speedY=0.5}) {samplerType=SAMPLERTYPE_Normal, texture=&T_Weap_Distort_N_02}
  n23 = Add(
   A=Multiply(A=ScalarParameter() {defaultValue=1, group=WPO, parameterName=WPO_B_Intensity}, B=LinearInterpolate(Alpha=@n18.R) {constB=1.5})
   B=Multiply(A=ScalarParameter() {defaultValue=1, group=WPO, parameterName=WPO_A_Intensity}, B=LinearInterpolate(Alpha=@n22.R) {constB=1.5})
  )
  n24 = OneMinus(None=ComponentMask(None=TextureCoordinate()) {g=true})
  n25 = Saturate(None=Multiply(A=Power(Base=@n24) {constExponent=1}))
  n26 = OneMinus(None=ComponentMask(None=TextureCoordinate()) {g=true})
  n27 = Clamp(None=Multiply(A=Power(Base=@n26) {constExponent=1})) {maxDefault=0.5}
  n28 = Time()
  n29 = Add(A=ComponentMask(None=TextureCoordinate()) {g=true}, B=Add(A=Multiply(A=@n17.WPO Speed, B=@n28), B=ParticleRandom()))
  n30 = Multiply(A=Saturate(None=Sine(None=@n29))) {constB=0.3}
  n31 = Add(
   A=Multiply(A=Multiply(A=@n23, B=@n25), B=Constant3Vector() {constant=<rgba 1.5 0 0 1>})
   B=Multiply(A=Multiply(A=@n27, B=@n30), B=Constant3Vector() {constant=<rgba 0 0 1 1>})
  ) {constB=0}
  n32 = ScalarParameter() {defaultValue=1, group=WPO, parameterName=WPO Intensity}
  n33 = DynamicParameter() {defaultValue=<rgba 1 1 50 1>, paramNames [Smoke Pan Speed, WPO Speed, Sphere Mask Radius, WPO Intensity]}
  Add(
   A=StaticSwitchParameter(
    True=Add(A=@n5.WorldPositionOffset, B=@n12)
    False=StaticSwitchParameter(True=@n5.WorldPositionOffset, False=@n13) {parameterName=Use_SplineThicken}
   ) {parameterName=Use_SplineThickenWithInterpolateWPO}
   B=Multiply(A=Transform(None=@n31), B=Multiply(A=@n32, B=@n33.WPO Intensity)) {constB=12}
  )
  n34 = CurveAtlasRowParameter(CurveTime=ComponentMask(None=TextureCoordinate()) {g=true}) {
   atlas=/Game/Athena/LinearColorCurves/A_Atlas_Turtleneck_Glider
   curve=/Game/Weapons/FORT_Melee/Pickaxe_Future_Samurai_Summer_Male/FX/Red/Curve_Pickaxe_Future_Samurai_Summer
   defaultValue=28
   group=Color
   parameterName=Curve
  }
  n35 = ScalarParameter() {defaultValue=1, group=Color, parameterName=GradientPower}
  n36 = Constant()
  n37 = ParticleRandom()
  n38 = MaterialFunctionCall(UVs=Add(A=TextureCoordinate(), B=AppendVector(A=@n36, B=@n37)), "Rotation Angle (0-1)"=Constant() {r=-0.25}) {materialFunction=&CustomRotator}
  n39 = ScalarParameter() {defaultValue=2, group=Smoke, parameterName=Smoke Tex B_ScaleU}
  n40 = ScalarParameter() {defaultValue=0.4, group=Smoke, parameterName=Smoke Tex B_ScaleV}
  n41 = Add(A=@n37, B=TextureCoordinate())
  n42 = AppendVector(
   A=ScalarParameter() {defaultValue=1, group=Smoke, parameterName=Distortion_ScaleU}
   B=ScalarParameter() {defaultValue=1, group=Smoke, parameterName=Distortion_ScaleV}
  )
  n43 = ComponentMask(None=TextureSample(UVs=Multiply(A=@n41, B=@n42)) {samplerType=SAMPLERTYPE_Normal, texture=&T_Weap_Distort_N_02}.RGB) {g=true, r=true}
  n44 = Add(
   A=MaterialFunctionCall(UVs=@n38.Rotated Values, Texture Scale=AppendVector(A=@n39, B=@n40)) {materialFunction=&ScaleUVsByCenter}.UVs
   B=Multiply(A=@n43, B=ScalarParameter() {group=Smoke, parameterName=Distort B Intensity}) {constB=0.1}
  )
  n45 = DynamicParameter() {defaultValue=<rgba 1 1 50 1>, paramNames [Smoke Pan Speed, WPO Speed, Sphere Mask Radius, WPO Intensity]}
  n46 = AppendVector(A=Multiply(A=ScalarParameter() {defaultValue=0.3, group=Smoke, parameterName=Smoke Tex B Pan}, B=@n45.Smoke Pan Speed), B=Constant())
  n47 = ChannelMaskParameter(None=TextureSampleParameter2D(UVs=Panner(Coordinate=@n44, Speed=@n46)) {group=Smoke, parameterName=Smoke Tex B, texture=/Game/Effects/Fort_Effects/Textures/Smoke/T_Smoke_Trail_Soft_2}.RGBA) {group=Smoke, parameterName=Smoke B Channel}
  n48 = ScalarParameter() {defaultValue=1, group=Smoke, parameterName=Smoke Tex A_ScaleU}
  n49 = ScalarParameter() {defaultValue=0.6, group=Smoke, parameterName=Smoke Tex A_ScaleV}
  n50 = Add(
   A=MaterialFunctionCall(UVs=@n38.Rotated Values, Texture Scale=AppendVector(A=@n48, B=@n49)) {materialFunction=&ScaleUVsByCenter}.UVs
   B=Multiply(A=@n43, B=ScalarParameter() {defaultValue=0.2, group=Smoke, parameterName=Distort A Intensity}) {constB=0.1}
  )
  n51 = AppendVector(A=Multiply(A=@n45.Smoke Pan Speed, B=ScalarParameter() {defaultValue=0.4, group=Smoke, parameterName=Smoke Tex A Pan}), B=Constant())
  n52 = ChannelMaskParameter(None=TextureSampleParameter2D(UVs=Panner(Coordinate=@n50, Speed=@n51) {speedX=0.4}) {group=Smoke, parameterName=Smoke A Tex, texture=/Game/Effects/Fort_Effects/Textures/Smoke/T_Smoke_Trail_Soft_Chromatic}.RGBA) {group=Smoke, parameterName=Smoke A Channel}
  n53 = LinearInterpolate(
   A=VectorParameter() {group=Color, parameterName=Color B, defaultValue=<rgba 0.005 0.005 0.005 1>}.RGB
   B=VectorParameter() {group=Color, parameterName=Color A, defaultValue=<rgba 0.03 0.03 0.03 1>}.RGB
   Alpha=LinearInterpolate(A=@n47.R, B=@n52.R, Alpha=ScalarParameter() {group=Color, parameterName=Lerp Color})
  )
  n54 = StaticSwitchParameter(True=Multiply(A=Power(Base=@n34.G, Exp=@n35) {constExponent=1}, B=@n53), False=@n53) {group=Color, parameterName=UseGradientColor}
  n55 = ParticleColor()
  n56 = VectorParameter() {parameterName=Color_Morning, defaultValue=<rgba 0.05 0.05 0.05 1>}
  n57 = Constant() {r=1}
  n58 = VectorParameter() {parameterName=Color_Evening, defaultValue=<rgba 0.05 0.05 0.05 1>}
  n59 = VectorParameter() {parameterName=Color_Night, defaultValue=<rgba 0.5906 0.6 0.5813 0.7>}
  MaterialFunctionCall(Emissive=Multiply(
    A=Multiply(A=@n54, B=@n55.RGB)
    B=MaterialFunctionCall(Color_Morning=@n56.RGB, Color_Day=@n57, Color_Evening=@n58.RGB, Color_Night=@n59.RGB) {materialFunction=&Color_Time_of_Day}.Result
   )) {materialFunction=&MF_FX_EmissiveAdjustments}
  n60 = Multiply(
   A=Add(A=@n47.R, B=ScalarParameter() {group=Smoke_B, parameterName=Smoke_B_GrayAdjust_Add})
   B=ScalarParameter() {defaultValue=1, group=Smoke_B, parameterName=Smoke_B_GrayAdjust_Mul}
  )
  n61 = Saturate(None=Power(Base=Saturate(None=@n60), Exp=ScalarParameter() {defaultValue=1, group=Smoke_B, parameterName=Smoke_B_GrayAdjust_Power}))
  n62 = DynamicParameter() {
   parameterIndex=2
   defaultValue=<rgba 6 1 50 1>
   paramNames [SmokeSpeed/-/OpacityBoost, WPO Speed/-/SmokeBoost, Sphere Mask Radius, WPO Intensity]
  }
  n63 = Multiply(A=Add(A=@n52.R, B=@n62.WPO Speed/-/SmokeBoost), B=ScalarParameter() {defaultValue=1, group=Smke_A, parameterName=Smke_A_GrayAdjust_Mul})
  n64 = Saturate(None=Power(Base=Saturate(None=@n63), Exp=ScalarParameter() {defaultValue=1, group=Smke_A, parameterName=Smke_A_GrayAdjust_Power}))
  n65 = Add(
   A=Multiply(
    A=StaticSwitchParameter(True=@n61, False=@n47.R) {parameterName=UseAdditionalAdjust}
    B=ScalarParameter() {defaultValue=0.5, group=Smoke, parameterName=Intensity Tex B}
   )
   B=Multiply(
    A=StaticSwitchParameter(True=@n64, False=@n52.R) {parameterName=UseAdditionalAdjust}
    B=ScalarParameter() {defaultValue=0.5, group=Smoke, parameterName=Intensity Tex A}
   ) {constB=0}
  )
  n66 = TextureCoordinate()
  n67 = VectorParameter() {group=Fire Shape, parameterName=Fire_MacroShape_UVScale, sortPriority=5, defaultValue=<rgba 1 1 0 0>}
  n68 = ComponentMask(None=@n67.RGB) {g=true, r=true}
  n69 = MaterialFunctionCall(Float2=Multiply(A=@n66, B=@n68)) {materialFunction=&BreakOutFloat2Components}
  n70 = Multiply(A=@n69.R) {constB=0.5}
  n71 = ScalarParameter() {defaultValue=3, group=Fire Shape, parameterName=Edge Fade Value Top, sortPriority=3}
  n72 = ScalarParameter() {defaultValue=2, group=Fire Shape, parameterName=Edge Fade Value Bottom, sortPriority=4}
  n73 = Multiply(A=Power(Base=Sine(None=@n70), Exp=LinearInterpolate(A=@n71, B=@n72, Alpha=@n69.G) {constA=6, constB=2}) {constExponent=4}, B=@n69.G)
  n74 = VectorParameter() {group=BottomFade, parameterName=Bottom Fade_SphereMask_A_Scale, sortPriority=2, defaultValue=<rgba 1.2 1 0 0>}
  n75 = VectorParameter() {group=BottomFade, parameterName=Bottom Fade_SphereMask_B, sortPriority=3, defaultValue=<rgba 0.6 0.05 0 0>}
  n76 = SphereMask(
   A=Multiply(A=TextureCoordinate(), B=ComponentMask(None=@n74.RGB) {g=true, r=true})
   B=ComponentMask(None=@n75.RGB) {g=true, r=true}
   Radius=ScalarParameter() {defaultValue=0.95, group=BottomFade, parameterName=BottomFade_Radius, sortPriority=1}
  ) {attenuationRadius=0.95, hardnessPercent=0}
  n77 = LinearInterpolate(B=Saturate(None=Multiply(A=@n76) {constB=15}), Alpha=ScalarParameter() {defaultValue=1, group=BottomFade, parameterName=Bottom Fade}) {constA=1}
  n78 = ParticleRandom()
  n79 = Add(A=@n78, B=TextureCoordinate() {vTiling=1.5})
  n80 = ComponentMask(None=VectorParameter() {group=Distortion, parameterName=Distortion UV Scale, sortPriority=3, defaultValue=<rgba 1 1 0 0>}.RGB) {g=true, r=true}
  n81 = TextureSampleParameter2D(UVs=Panner(Coordinate=Multiply(A=@n79, B=@n80)) {speedY=1}) {
   group=Distortion
   parameterName=Distortion Normal
   samplerType=SAMPLERTYPE_Normal
   sortPriority=2
   texture=&T_Weap_Distort_N_02
  }
  n82 = ComponentMask(None=@n81.RGB) {g=true, r=true}
  n83 = ComponentMask(None=Multiply(A=@n66, B=@n68)) {g=true}
  n84 = LinearInterpolate(
   B=Divide(
    A=LinearInterpolate(A=@n82, Alpha=@n83) {constB=0}
    B=ScalarParameter() {defaultValue=3, group=Distortion, parameterName=UV_Distortion_Divsor, sortPriority=1}
   ) {constB=20}
   Alpha=ScalarParameter() {defaultValue=1, group=Distortion, parameterName=UV_Distortion_Lerp}
  )
  n85 = Add(A=@n78, B=TextureCoordinate() {uTiling=0.8, vTiling=0.9})
  n86 = ComponentMask(None=VectorParameter() {group=Fire Shape, parameterName=Fire_Micro_UVScale, sortPriority=8, defaultValue=<rgba 2 2 0 0>}.RGB) {g=true, r=true}
  n87 = Panner(Coordinate=MaterialFunctionCall(XY=Multiply(A=@n85, B=@n86)) {materialFunction=&Swizzle}.YX) {speedX=1.2, speedY=0.2}
  n88 = TextureSampleParameter2D(UVs=ShadingPathSwitch(Default=Add(A=@n84, B=@n87) {constB=0}, Mobile=@n87) {desc=Migrated from FeatureLevelSwitch}) {
   group=Fire Shape
   parameterName=Fire Micro Texture
   samplerType=SAMPLERTYPE_LinearColor
   sortPriority=6
   texture=&T_NoisePack_LinearColor
  }
  n89 = ChannelMaskParameter(None=@n88.RGBA) {group=Fire Shape, parameterName=Fire Micro Mask, sortPriority=7}
  n90 = Constant() {r=1}
  n91 = ScalarParameter() {defaultValue=1, parameterName=Fire Micro lerp}
  n92 = Power(
   Base=Multiply(A=LinearInterpolate(A=@n89.R, B=@n90, Alpha=@n91) {constAlpha=0.2}, B=Multiply(A=@n73, B=@n77)) {constA=1}
   Exp=LinearInterpolate(Alpha=@n69.G) {constA=2}
  )
  n93 = ScalarParameter() {defaultValue=1.5, group=Fire Shape, parameterName=Fire Split Contrast, sortPriority=1}
  n94 = OneMinus(None=Sine(None=Multiply(A=@n69.R) {constB=0.45}))
  n95 = LinearInterpolate(Alpha=@n81.R) {constA=0.5, constB=1.5}
  n96 = Round(None=Add(A=Power(Base=@n69.G, Exp=@n93), B=Power(Base=@n94, Exp=@n95) {constExponent=0.5}))
  n97 = ScalarParameter() {defaultValue=0.2, group=Fire Shape, parameterName=Fire Split}
  n98 = DynamicParameter() {
   parameterIndex=2
   defaultValue=<rgba 6 1 50 1>
   paramNames [SmokeSpeed/-/OpacityBoost, WPO Speed/-/SmokeBoost, Sphere Mask Radius, WPO Intensity]
  }
  n99 = Multiply(A=Multiply(A=Multiply(A=@n65, B=@n92), B=LinearInterpolate(B=@n96, Alpha=@n97) {constA=1, constB=0}), B=@n98.SmokeSpeed/-/OpacityBoost) {constB=3}
  n100 = Multiply(A=DepthFade(Opacity=@n99, FadeDistance=ScalarParameter() {defaultValue=5, group=Opacity, parameterName=DepthFadeDistance}))
  n101 = CameraVectorWS()
  n102 = Saturate(None=Multiply(A=DotProduct(A=@n101, B=@n5.Normal), B=ScalarParameter() {defaultValue=3, parameterName=FadeSplineThicken_Mul}))
  n103 = ScalarParameter() {defaultValue=3, parameterName=FadeSplineThicken_Power}
  n104 = ScalarParameter() {defaultValue=0.5, parameterName=FadeBasedOnViewAngle}
  n105 = Multiply(
   A=Multiply(A=ShadingPathSwitch(Default=@n100, Mobile=@n99) {desc=Migrated from FeatureLevelSwitch}, B=ParticleColor().A)
   B=StaticSwitchParameter(
    True=Power(Base=@n102, Exp=@n103)
    False=MaterialFunctionCall("FadeSensitivity (optional)"=@n104) {materialFunction=&FadeBasedOnViewAngle}.Result
   ) {parameterName=Use_SplineThicken}
  )
  n106 = DynamicParameter() {defaultValue=<rgba 1 1 50 1>, paramNames [Smoke Pan Speed, WPO Speed, Sphere Mask Radius, WPO Intensity]}
  n107 = SphereMask(A=WorldPosition().XYZ, B=ParticlePositionWS(), Radius=@n106.Sphere Mask Radius) {attenuationRadius=50, hardnessPercent=0.2}
  Multiply(
   A=MaterialFunctionCall(Value_Morning=@n56.A, Value_Day=@n57, Value_Evening=@n58.A, Value_Night=@n59.A) {materialFunction=&Scalar_Time_of_Day}.Result
   B=StaticSwitchParameter(True=Multiply(A=@n105, B=@n107), False=@n105) {parameterName=Use Sphere mask}
  )
  n108 = LinearInterpolate(
   A=ScalarParameter() {defaultValue=1, group=WPO, parameterName=WPO UV Scale Mult}
   B=DynamicParameter() {parameterIndex=1, defaultValue=<rgba 1 1 50 1>, paramNames [WPO UV Scale, "", "", ""]}.WPO UV Scale
   Alpha=ScalarParameter() {group=WPO, parameterName=Use Dynamic WPO UV Scale}
  )
  Multiply(A=@n108, B=@n20)
  Multiply(A=@n15, B=@n108)
  Saturate(None=@n99)
  ScalarParameter() {defaultValue=1, parameterName=AngleBasedFadeReplacement}
  ScalarParameter() {group=Smke_A, parameterName=Smke_A_GrayAdjust_Add}
 ]
}
